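# Physics asset for the F_MED_Cadet backpack cosmetic, rigging a small jointed figure with head, body, arms, legs and two zippers as dyn_ bones.
# Each bone has one small capsule body, with larger radii on the head and legs.
# Constraints chain root to pack to upper body, which branches to head, arms and lower body with legs; zippers hang off the right arm and lower body.
# Angular limits are tight, with looser swing on limbs.
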PhysicsAsset /BRCosmetics/Accessories/FORT_Backpacks/Backpack_F_MED_Cadet/Meshes/F_MED_Cadet_Pack_Physics
bodies {
 dyn_arm_L_01 [{shapeType=Capsule, length=1, radius=1, shapeName=dyn_arm_L_01_capsule}]
 dyn_arm_L_02 [{shapeType=Capsule, length=1, radius=1, shapeName=dyn_arm_L_02_capsule}]
 dyn_arm_R_01 [{shapeType=Capsule, length=1, radius=1, shapeName=dyn_arm_R_01_capsule}]
 dyn_arm_R_02 [{shapeType=Capsule, length=1, radius=1, shapeName=dyn_arm_R_02_capsule}]
 dyn_head [{shapeType=Capsule, length=1, radius=3, shapeName=dyn_head_capsule, center=<0.6583 0 0>}]
 dyn_leg_L_01 [{shapeType=Capsule, length=1, radius=1.5, shapeName=dyn_leg_L_01_capsule, center=<0.7451 0 0>, rotation=<rot 0 0.0005 89.9998>}]
 dyn_leg_L_02 [{shapeType=Capsule, length=1, radius=2, shapeName=dyn_leg_L_02_capsule, rotation=<rot 0 0.0005 89.9998>}]
 dyn_leg_R_01 [{shapeType=Capsule, length=1, radius=1.5, shapeName=dyn_leg_R_01_capsule, rotation=<rot 0.0002 -0.0004 110>}]
 dyn_leg_R_02 [{shapeType=Capsule, length=1, radius=2, shapeName=dyn_leg_R_02_capsule, rotation=<rot 0.0002 -0.0004 110>}]
 dyn_lowerBody [{shapeType=Capsule, length=1, radius=1, shapeName=dyn_lowerBody_capsule}]
 dyn_upperBody [{shapeType=Capsule, length=1, radius=1, shapeName=dyn_upperBody_capsule}]
 dyn_zipper_arm [{shapeType=Capsule, length=1, radius=1, shapeName=dyn_zipper_arm_capsule}]
 dyn_zipper_body [{shapeType=Capsule, length=1, radius=1, shapeName=dyn_zipper_body_capsule}]
 pack [{shapeType=Capsule, length=1, radius=1, shapeName=pack_capsule}]
 root [{shapeType=Capsule, length=1, radius=1, shapeName=root_capsule}]
}
constraints [
 {bone1Name=dyn_zipper_body, bone2Name=dyn_lowerBody, swing1LimitDegrees=10, swing2LimitDegrees=10, twistLimitDegrees=10}
 {bone1Name=pack, bone2Name=root, swing1LimitDegrees=10, swing2LimitDegrees=10, twistLimitDegrees=10}
 {bone1Name=dyn_upperBody, bone2Name=pack, swing1LimitDegrees=10, swing2LimitDegrees=10, twistLimitDegrees=10}
 {bone1Name=dyn_head, bone2Name=dyn_upperBody, swing1LimitDegrees=10, swing2LimitDegrees=10, twistLimitDegrees=10}
 {bone1Name=dyn_arm_R_01, bone2Name=dyn_upperBody, swing1LimitDegrees=20, swing2LimitDegrees=20, twistLimitDegrees=10}
 {bone1Name=dyn_arm_R_02, bone2Name=dyn_arm_R_01, swing1LimitDegrees=20, swing2LimitDegrees=20, twistLimitDegrees=10}
 {bone1Name=dyn_zipper_arm, bone2Name=dyn_arm_R_01, swing1LimitDegrees=20, swing2LimitDegrees=20, twistLimitDegrees=10}
 {bone1Name=dyn_arm_L_01, bone2Name=dyn_upperBody, swing1LimitDegrees=20, swing2LimitDegrees=20, twistLimitDegrees=10}
 {bone1Name=dyn_arm_L_02, bone2Name=dyn_arm_L_01, swing1LimitDegrees=20, swing2LimitDegrees=20, twistLimitDegrees=10}
 {bone1Name=dyn_lowerBody, bone2Name=dyn_upperBody, swing1LimitDegrees=10, swing2LimitDegrees=10, twistLimitDegrees=10}
 {bone1Name=dyn_leg_R_01, bone2Name=dyn_lowerBody, swing1LimitDegrees=20, swing2LimitDegrees=20, twistLimitDegrees=10}
 {bone1Name=dyn_leg_R_02, bone2Name=dyn_leg_R_01, swing1LimitDegrees=20, swing2LimitDegrees=20, twistLimitDegrees=10}
 {bone1Name=dyn_leg_L_01, bone2Name=dyn_lowerBody, swing1LimitDegrees=20, swing2LimitDegrees=20, twistLimitDegrees=10}
 {bone1Name=dyn_leg_L_02, bone2Name=dyn_leg_L_01, swing1LimitDegrees=20, swing2LimitDegrees=20, twistLimitDegrees=10}
]
